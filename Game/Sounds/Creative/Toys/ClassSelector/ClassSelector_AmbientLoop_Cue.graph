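# Looping ambient sound cue for the Creative class selector toy, classed as a Creative gadget sound.
# A single wave player loops the ambient wave at reduced volume, with no mixing or randomization.
# Cue-level overrides set a short-range falloff with stereo spread, enable focus interpolation that quiets the sound when unfocused, and cap concurrent instances to a small count.
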SoundCue /Game/Sounds/Creative/Toys/ClassSelector/ClassSelector_AmbientLoop_Cue
root=WavePlayer_0
nodes [{id=WavePlayer_0, properties {bLooping=true, soundWaveAssetPtr=&/ClassSelector_AmbientLoop_01}}]
properties {
 bOverrideAttenuation=true
 bOverrideConcurrency=true
 maxDistance=1200
 soundClassObject=&Fort_Gadgets_Creative
 volumeMultiplier=0.3
 attenuationOverrides {bEnableFocusInterpolation=true, falloffDistance=1200, nonFocusVolumeAttenuation=0.5, stereoSpread=500, attenuationShapeExtents {x=0}}
 concurrencyOverrides {maxCount=2}
}
