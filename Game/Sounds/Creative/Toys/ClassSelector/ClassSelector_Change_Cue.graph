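# Sound cue for the change sound of the Class Selector creative toy, in the creative gadgets sound class.
# A Random node picks between two equally weighted wave variations, played at reduced volume.
# Attenuation is overridden on the cue itself, with its own falloff and focus settings in place of the referenced MoveTool attenuation asset.
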SoundCue /Game/Sounds/Creative/Toys/ClassSelector/ClassSelector_Change_Cue
root=Random_0
nodes [
 {id=Random_0, children [WavePlayer_0,WavePlayer_1], properties {weights [1,1]}}
 {id=WavePlayer_0, properties {soundWaveAssetPtr=&/Class_Selector_Change_01}}
 {id=WavePlayer_1, properties {soundWaveAssetPtr=&/Class_Selector_Change_02}}
]
properties {
 attenuationSettings=/Game/Sounds/Creative/Attenuations/Fort_Gadget_MoveTool_Att
 bOverrideAttenuation=true
 maxDistance=2000
 soundClassObject=&Fort_Gadgets_Creative
 volumeMultiplier=0.7
 attenuationOverrides {falloffDistance=1700, focusAzimuth=15, nonFocusAzimuth=45, nonFocusVolumeAttenuation=0.5, attenuationShapeExtents {x=300}}
 concurrencyOverrides {maxCount=1, resolutionRule=PreventNew}
}
