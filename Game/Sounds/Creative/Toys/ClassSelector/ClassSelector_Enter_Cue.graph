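# Sound cue for entering the Creative class selector toy, routed to the Creative gadgets sound class at reduced volume.
# A single wave player feeds a modulator with pitch variation pinned off.
# Inline attenuation overrides supersede the referenced MoveTool attenuation asset, adding custom falloff and focus values.
# Concurrency override values specify one instance that prevents new plays, though the flag enabling that override is not set.
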SoundCue /Game/Sounds/Creative/Toys/ClassSelector/ClassSelector_Enter_Cue
root=Modulator_0
nodes [
 {id=Modulator_0, children [WavePlayer_0], properties {pitchMax=1, pitchMin=1}}
 {id=WavePlayer_0, properties {soundWaveAssetPtr=&/Class_Selector_Enter_01}}
]
properties {
 attenuationSettings=/Game/Sounds/Creative/Attenuations/Fort_Gadget_MoveTool_Att
 bOverrideAttenuation=true
 maxDistance=2000
 soundClassObject=&Fort_Gadgets_Creative
 volumeMultiplier=0.65
 attenuationOverrides {falloffDistance=1700, focusAzimuth=15, nonFocusAzimuth=45, nonFocusVolumeAttenuation=0.5, attenuationShapeExtents {x=300}}
 concurrencyOverrides {maxCount=1, resolutionRule=PreventNew}
}
